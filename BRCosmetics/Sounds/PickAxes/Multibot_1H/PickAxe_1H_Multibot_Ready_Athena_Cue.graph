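# Sound cue for the ready sound of the one-handed Multibot pickaxe cosmetic, using its swing samples at close range and Fire Axe distant swing samples at medium range.
# A mixer layers two branches, each running through its own attenuation, modulator and quality-level switch.
# The quality switches offer progressively smaller variation pools, from random sets down to a single wave.
# Playback is limited per owner, stopping the oldest instance.
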SoundCue /BRCosmetics/Sounds/PickAxes/Multibot_1H/PickAxe_1H_Multibot_Ready_Athena_Cue
root=Mixer_0
nodes [
 {id=Attenuation_0, children [Modulator_1], properties {attenuationSettings=/Game/Sounds/Attenuation/Fort_Weapon_Pickaxe_Close_Att}}
 {id=Attenuation_1, children [Modulator_0], properties {attenuationSettings=/Game/Sounds/Attenuation/Fort_Weapon_Pickaxe_Medium_Att}}
 {id=Mixer_0, children [Attenuation_0, Attenuation_1], properties {inputVolume [1,1]}}
 {id=Modulator_0, children [QualityLevel_1]}
 {id=Modulator_1, children [QualityLevel_0]}
 {id=QualityLevel_0, children [Random_3, Random_5, WavePlayer_27, WavePlayer_27]}
 {id=QualityLevel_1, children [Random_0, Random_1, WavePlayer_12, WavePlayer_12]}
 {id=Random_0, children [WavePlayer_1,WavePlayer_2,WavePlayer_4,WavePlayer_6,WavePlayer_7,WavePlayer_8], properties {weights [1,1,1,1,1,1]}}
 {id=Random_1, children [WavePlayer_10, WavePlayer_9, WavePlayer_11], properties {weights [1,1,1]}}
 {id=Random_3, children [WavePlayer_13, WavePlayer_16, WavePlayer_18, WavePlayer_19], properties {weights [1,1,1,1]}}
 {id=Random_5, children [WavePlayer_20, WavePlayer_21, WavePlayer_26], properties {weights [1,1,1]}}
 {id=WavePlayer_1, properties {soundWaveAssetPtr=/Game/Athena/Sounds/Weapons/Fire_Axe_Swing_Distant_04}}
 {id=WavePlayer_10, properties {soundWaveAssetPtr=&Fire_Axe_Swing_Distant_04}}
 {id=WavePlayer_11, properties {soundWaveAssetPtr=/Game/Athena/Sounds/Weapons/Fire_Axe_Swing_Distant_06}}
 {id=WavePlayer_12, properties {soundWaveAssetPtr=&Fire_Axe_Swing_Distant_06}}
 {id=WavePlayer_13, properties {soundWaveAssetPtr=&/SFX_Pickaxe_Multibot_Swing_01}}
 {id=WavePlayer_16, properties {soundWaveAssetPtr=&/SFX_Pickaxe_Multibot_Swing_02}}
 {id=WavePlayer_18, properties {soundWaveAssetPtr=&/SFX_Pickaxe_Multibot_Swing_03}}
 {id=WavePlayer_19, properties {soundWaveAssetPtr=&/SFX_Pickaxe_Multibot_Swing_04}}
 {id=WavePlayer_2, properties {soundWaveAssetPtr=/Game/Athena/Sounds/Weapons/Fire_Axe_Swing_Distant_01}}
 {id=WavePlayer_20, properties {soundWaveAssetPtr=&SFX_Pickaxe_Multibot_Swing_01}}
 {id=WavePlayer_21, properties {soundWaveAssetPtr=&SFX_Pickaxe_Multibot_Swing_02}}
 {id=WavePlayer_26, properties {soundWaveAssetPtr=&SFX_Pickaxe_Multibot_Swing_03}}
 {id=WavePlayer_27, properties {soundWaveAssetPtr=&SFX_Pickaxe_Multibot_Swing_01}}
 {id=WavePlayer_4, properties {soundWaveAssetPtr=/Game/Athena/Sounds/Weapons/Fire_Axe_Swing_Distant_02}}
 {id=WavePlayer_6, properties {soundWaveAssetPtr=/Game/Athena/Sounds/Weapons/Fire_Axe_Swing_Distant_03}}
 {id=WavePlayer_7, properties {soundWaveAssetPtr=/Game/Athena/Sounds/Weapons/Fire_Axe_Swing_Distant_05}}
 {id=WavePlayer_8, properties {soundWaveAssetPtr=&Fire_Axe_Swing_Distant_06}}
 {id=WavePlayer_9, properties {soundWaveAssetPtr=&Fire_Axe_Swing_Distant_03}}
]
properties {
 audioPropertiesSheet=/BRCosmetics/Sounds/Settings/PropertySheets/Pickaxe/APS_Cosmetic_Pickaxe_Ready
 bOverrideConcurrency=true
 bPrimeOnLoad=true
 maxDistance=3000
 soundClassObject=&Fort_WeaponMelee
 volumeMultiplier=0.3
 concurrencyOverrides {bLimitToOwner=true, maxCount=2, resolutionRule=StopOldest}
 concurrencySet [&Player_Melee_Weapon_Conc]
}
