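# Sound cue for the one-handed Multibot pickaxe swing.
# A mixer layers a close-attenuated branch of Multibot swing waves with a medium-attenuated branch reusing Fire Axe distant swing waves.
# Each branch runs through a modulator into a quality switch, which selects random pools of varying size or a single fixed wave to scale variation by quality level.
# Per-owner concurrency caps overlapping swings by stopping the oldest, and the cue primes on load.
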SoundCue /BRCosmetics/Sounds/PickAxes/Multibot_1H/PickAxe_1H_Multibot_Swing_Athena_Cue
root=Mixer_1
nodes [
 {id=Attenuation_2, children [Modulator_3], properties {attenuationSettings=/Game/Sounds/Attenuation/Fort_Weapon_Pickaxe_Close_Att}}
 {id=Attenuation_3, children [Modulator_2], properties {attenuationSettings=/Game/Sounds/Attenuation/Fort_Weapon_Pickaxe_Medium_Att}}
 {id=Mixer_1, children [Attenuation_2, Attenuation_3], properties {inputVolume [1,1]}}
 {id=Modulator_2, children [QualityLevel_3]}
 {id=Modulator_3, children [QualityLevel_0]}
 {id=QualityLevel_0, children [Random_6, Random_7, WavePlayer_37, WavePlayer_37]}
 {id=QualityLevel_3, children [Random_3, Random_5, WavePlayer_29, WavePlayer_29]}
 {
  id=Random_3
  children [WavePlayer_13, WavePlayer_16, WavePlayer_18, WavePlayer_19, WavePlayer_20, WavePlayer_21]
  properties {weights [1,1,1,1,1,1]}
 }
 {id=Random_5, children [WavePlayer_27, WavePlayer_26, WavePlayer_28], properties {weights [1,1,1]}}
 {id=Random_6, children [WavePlayer_30, WavePlayer_31, WavePlayer_32, WavePlayer_33], properties {weights [1,1,1,1]}}
 {id=Random_7, children [WavePlayer_34, WavePlayer_35, WavePlayer_36], properties {weights [1,1,1]}}
 {id=WavePlayer_13, properties {soundWaveAssetPtr=/Game/Athena/Sounds/Weapons/Fire_Axe_Swing_Distant_04}}
 {id=WavePlayer_16, properties {soundWaveAssetPtr=/Game/Athena/Sounds/Weapons/Fire_Axe_Swing_Distant_01}}
 {id=WavePlayer_18, properties {soundWaveAssetPtr=/Game/Athena/Sounds/Weapons/Fire_Axe_Swing_Distant_02}}
 {id=WavePlayer_19, properties {soundWaveAssetPtr=/Game/Athena/Sounds/Weapons/Fire_Axe_Swing_Distant_03}}
 {id=WavePlayer_20, properties {soundWaveAssetPtr=/Game/Athena/Sounds/Weapons/Fire_Axe_Swing_Distant_05}}
 {id=WavePlayer_21, properties {soundWaveAssetPtr=/Game/Athena/Sounds/Weapons/Fire_Axe_Swing_Distant_06}}
 {id=WavePlayer_26, properties {soundWaveAssetPtr=&Fire_Axe_Swing_Distant_03}}
 {id=WavePlayer_27, properties {soundWaveAssetPtr=&Fire_Axe_Swing_Distant_04}}
 {id=WavePlayer_28, properties {soundWaveAssetPtr=&Fire_Axe_Swing_Distant_06}}
 {id=WavePlayer_29, properties {soundWaveAssetPtr=&Fire_Axe_Swing_Distant_06}}
 {id=WavePlayer_30, properties {soundWaveAssetPtr=&/SFX_Pickaxe_Multibot_Swing_01}}
 {id=WavePlayer_31, properties {soundWaveAssetPtr=&/SFX_Pickaxe_Multibot_Swing_02}}
 {id=WavePlayer_32, properties {soundWaveAssetPtr=&/SFX_Pickaxe_Multibot_Swing_03}}
 {id=WavePlayer_33, properties {soundWaveAssetPtr=&/SFX_Pickaxe_Multibot_Swing_04}}
 {id=WavePlayer_34, properties {soundWaveAssetPtr=&SFX_Pickaxe_Multibot_Swing_01}}
 {id=WavePlayer_35, properties {soundWaveAssetPtr=&SFX_Pickaxe_Multibot_Swing_02}}
 {id=WavePlayer_36, properties {soundWaveAssetPtr=&SFX_Pickaxe_Multibot_Swing_03}}
 {id=WavePlayer_37, properties {soundWaveAssetPtr=&SFX_Pickaxe_Multibot_Swing_01}}
]
properties {
 audioPropertiesSheet=/BRCosmetics/Sounds/Settings/PropertySheets/Pickaxe/APS_Cosmetic_Pickaxe_Swing
 bOverrideConcurrency=true
 bPrimeOnLoad=true
 maxDistance=3000
 soundClassObject=&Fort_WeaponMelee_PSM
 volumeMultiplier=0.3
 concurrencyOverrides {bLimitToOwner=true, maxCount=2, resolutionRule=StopOldest}
 concurrencySet [&Player_Melee_Weapon_Conc]
}
